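# Physics asset for the Cyber Fu backpack cosmetic, simulating its nunchaku chain and dangling zipper around a capsule pack body.
# The nunchakus body is fully locked to the pack and anchors two runs of small capsule links, seven in strand A and five in strand B.
# Most links swing with twist locked, and a final constraint joins both strand ends into a closed loop.
# The zipper hangs separately from the pack with twist locked.
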PhysicsAsset /BRCosmetics/Accessories/FORT_Backpacks/F_MED_CyberFu_Backpack/Meshes/F_MED_Cyber_Fu_Pack_Physics
bodies {
 dyn__nunchakus [{shapeType=Capsule, length=24, radius=4, rotation=<rot 0 0 90>}]
 dyn_chain_a01 [{shapeType=Capsule, length=1, radius=1, center=<1.0655 0 0>, rotation=<rot 90 0 0>}]
 dyn_chain_a02 [{shapeType=Capsule, length=1, radius=1, center=<0.7213 0 0>, rotation=<rot 90 0 0>}]
 dyn_chain_a03 [{shapeType=Capsule, length=1, radius=1, center=<0.6915 0 0>, rotation=<rot 90 0 0>}]
 dyn_chain_a04 [{shapeType=Capsule, length=1, radius=1, center=<0.443 0 0>, rotation=<rot 90 0 0>}]
 dyn_chain_a05 [{shapeType=Capsule, length=1, radius=1, center=<0.5189 0 0>, rotation=<rot 90 0 0>}]
 dyn_chain_a06 [{shapeType=Capsule, length=1, radius=1, center=<0.5484 0 0>, rotation=<rot 90 0 0>}]
 dyn_chain_a07 [{shapeType=Capsule, length=1, radius=1, center=<0.8749 0 0>, rotation=<rot 90 0 0>}]
 dyn_chain_b01 [{shapeType=Capsule, length=1, radius=1, center=<0.654 0 0>, rotation=<rot 90 0 0>}]
 dyn_chain_b02 [{shapeType=Capsule, length=1, radius=1, center=<0.876 0 0>, rotation=<rot 90 0 0>}]
 dyn_chain_b03 [{shapeType=Capsule, length=1, radius=1, center=<0.531 0 0>, rotation=<rot 90 0 0>}]
 dyn_chain_b04 [{shapeType=Capsule, length=1, radius=1, center=<0.5598 0 0>, rotation=<rot 90 0 0>}]
 dyn_chain_b05 [{shapeType=Capsule, length=1, radius=1, rotation=<rot 90 0 0>}]
 dyn_zipper [{shapeType=Capsule, length=1.8202, radius=1.1879, center=<0.8634 0 0.0001>, rotation=<rot 90 1.8984 1.8983>}]
 pack [{shapeType=Capsule, length=18, radius=8, center=<0 -4.7343 -5.2083>}]
}
constraints [
 {
  bone1Name=dyn__nunchakus
  bone2Name=pack
  swing1LimitDegrees=45
  swing1Motion=Locked
  swing2LimitDegrees=45
  swing2Motion=Locked
  twistLimitDegrees=45
  twistMotion=Locked
 }
 {bone1Name=dyn_chain_a01, bone2Name=dyn__nunchakus, swing1LimitDegrees=45, swing2LimitDegrees=45, twistLimitDegrees=45, twistMotion=Locked}
 {bone1Name=dyn_chain_a02, bone2Name=dyn_chain_a01, swing1LimitDegrees=45, swing2LimitDegrees=45, twistLimitDegrees=45, twistMotion=Locked}
 {bone1Name=dyn_chain_a03, bone2Name=dyn_chain_a02, swing1LimitDegrees=45, swing2LimitDegrees=45, twistLimitDegrees=45, twistMotion=Locked}
 {bone1Name=dyn_chain_a04, bone2Name=dyn_chain_a03, swing1LimitDegrees=45, swing2LimitDegrees=45, twistLimitDegrees=45, twistMotion=Locked}
 {bone1Name=dyn_chain_a05, bone2Name=dyn_chain_a04, swing1LimitDegrees=45, swing2LimitDegrees=45, twistLimitDegrees=45, twistMotion=Locked}
 {bone1Name=dyn_chain_a06, bone2Name=dyn_chain_a05, swing1LimitDegrees=45, swing2LimitDegrees=45, twistLimitDegrees=45, twistMotion=Locked}
 {bone1Name=dyn_chain_a07, bone2Name=dyn_chain_a06, swing1LimitDegrees=45, swing2LimitDegrees=45, twistLimitDegrees=15}
 {bone1Name=dyn_chain_b01, bone2Name=dyn__nunchakus, swing1LimitDegrees=45, swing2LimitDegrees=45, twistLimitDegrees=45, twistMotion=Locked}
 {bone1Name=dyn_chain_b02, bone2Name=dyn_chain_b01, swing1LimitDegrees=45, swing2LimitDegrees=45, twistLimitDegrees=45, twistMotion=Locked}
 {bone1Name=dyn_chain_b03, bone2Name=dyn_chain_b02, swing1LimitDegrees=45, swing2LimitDegrees=45, twistLimitDegrees=45, twistMotion=Locked}
 {bone1Name=dyn_chain_b04, bone2Name=dyn_chain_b03, swing1LimitDegrees=45, swing2LimitDegrees=45, twistLimitDegrees=45, twistMotion=Locked}
 {
  bone1Name=dyn_chain_b05
  bone2Name=dyn_chain_b04
  swing1LimitDegrees=45
  swing1Motion=Locked
  swing2LimitDegrees=45
  swing2Motion=Locked
  twistLimitDegrees=45
  twistMotion=Locked
 }
 {bone1Name=dyn_zipper, bone2Name=pack, swing1LimitDegrees=45, swing2LimitDegrees=60, twistLimitDegrees=45, twistMotion=Locked}
 {bone1Name=dyn_chain_b05, bone2Name=dyn_chain_a07, swing1LimitDegrees=45, swing2LimitDegrees=45, twistLimitDegrees=45}
]
